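# Sound cue for the glider open sound of the Black Monday Cape glider cosmetic.
# After a short fixed delay, a mixer layers the cape open wave with a quieter open layer borrowed from the Black Monday Male glider.
# Glider open/close attenuation and concurrency, a cosmetic glider open property sheet and the glider land sound class govern playback.
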SoundCue /BRCosmetics/Sounds/Gliders/BlackMondayCape/Glider_BlackMonday_Cape_Open_Cue
root=Delay_1
nodes [
 {id=Delay_1, children [Mixer_1], properties {delayMax=0.07, delayMin=0.07}}
 {id=Mixer_1, children [WavePlayer_0,WavePlayer_4], properties {inputVolume [1,0.65]}}
 {id=WavePlayer_0, properties {soundWaveAssetPtr=&/Glider_BlackMonday_Cape_Open_01}}
 {id=WavePlayer_4, properties {soundWaveAssetPtr=/BRCosmetics/Sounds/Gliders/BlackMondayMale/Glider_BlackMonday_Male_OpenLayer_01}}
]
properties {
 attenuationSettings=/Game/Sounds/Attenuation/Fort_Player_Glider_OpenClose_Athena_Att
 audioPropertiesSheet=/BRCosmetics/Sounds/Settings/PropertySheets/Glider/APS_Cosmetic_Glider_Open
 maxDistance=10000
 soundClassObject=&Fort_Gliders_Land
 volumeMultiplier=0.7
 concurrencySet [&Athena_GliderOpenClose_Conc]
}
